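# Material function in the FlakeSlide backpack's folder that adjusts normal, base color, specular and metallic on incoming material attributes.
# A TechArtMask input dotted with a channel-selector parameter (blue by default) blends in the normal and color edits.
# Normals get XY scaled, Z set to one, then renormalized; color gets multiplied, then saturated through negated Desaturation.
# Specular and metallic use plain, unmasked multipliers before all four attributes are written back.
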MaterialFunction /BRCosmetics/Accessories/FORT_Backpacks/Backpack_M_MED_FlakeSlide/Materials/MF_NormalBoost
graph {nodes [
  n1 = GetMaterialAttributes(None=FunctionInput() {inputName=In_MaterialAttribute, inputType=FunctionInput_MaterialAttributes}) {attributeGetTypes [
    #1
    #2
    #3
    #4
   ]}
  n2 = AppendVector(
   A=Multiply(A=ComponentMask(None=@n1.Normal) {g=true, r=true}, B=ScalarParameter() {defaultValue=1, group=Normals, parameterName=NormalMultiplier})
   B=Constant() {r=1}
  )
  n3 = NamedRerouteUsage()
  n4 = VectorParameter() {parameterName=TechArtMaskChannel, defaultValue=<rgba 0 0 1 0>}
  n5 = SetMaterialAttributes!(
   MaterialAttributes=@n1.MaterialAttributes
   Normal=LinearInterpolate(A=@n1.Normal, B=Normalize(VectorInput=@n2), Alpha=DotProduct(A=@n3, B=@n4.RGB))
   BaseColor=NamedRerouteUsage()
   Specular=NamedRerouteUsage()
   Metallic=NamedRerouteUsage()
  ) {attributeSetTypes [
    #1
    #2
    #3
    #4
   ]}
  FunctionOutput(None=@n5) {outputName=MaterialAttribute}
  NamedRerouteDeclaration(None=@n1.BaseColor) {name=In_BaseColor, nodeColor=<rgba 0.7109 0 1 1>}
  n6 = NamedRerouteUsage()
  n7 = Multiply(A=@n6, B=ScalarParameter() {defaultValue=1, parameterName=BaseColorMultiplier})
  n8 = Multiply(A=ScalarParameter() {parameterName=Saturation}) {constB=-1}
  n9 = NamedRerouteUsage()
  n10 = VectorParameter() {parameterName=TechArtMaskChannel, defaultValue=<rgba 0 0 1 0>}
  NamedRerouteDeclaration(None=LinearInterpolate(A=@n6, B=Desaturation(None=@n7, Fraction=@n8), Alpha=DotProduct(A=@n9, B=@n10.RGB))) {name=Out_BaseColor, nodeColor=<rgba 0 1 0.3906 1>}
  NamedRerouteDeclaration(None=FunctionInput() {inputName=In_TechArtMask}) {name=TechArtMask, nodeColor=<rgba 1 0 0.2344 1>}
  NamedRerouteDeclaration(None=Multiply(
    A=NamedRerouteDeclaration(None=@n1.Specular) {name=In_Specular, nodeColor=<rgba 0.1245 1 1 1>}
    B=ScalarParameter() {defaultValue=1, parameterName=SpecularMultiplier}
   )) {name=Out_Specular, nodeColor=<rgba 0 1 0.3203 1>}
  NamedRerouteDeclaration(None=@n1.Metallic) {name=In_Metallic, nodeColor=<rgba 0.8047 1 0 1>}
  NamedRerouteDeclaration(None=Multiply(A=NamedRerouteUsage(), B=ScalarParameter() {defaultValue=1, parameterName=MetallicMultiplier})) {name=Out_Metallic, nodeColor=<rgba 0 1 0.5547 1>}
 ]}
