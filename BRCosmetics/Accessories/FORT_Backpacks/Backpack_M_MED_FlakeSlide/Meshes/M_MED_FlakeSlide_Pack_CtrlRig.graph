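# Control Rig for the FlakeSlide backpack cosmetic that adjusts the pose of its pack bone depending on a skydive flag.
# When skydive is false, a branch runs a Modify Transforms node that applies a rotation and small translation offset to the pack.
# The true branch is unconnected, and a second rotation-only Modify Transforms node sits unwired.
# The hierarchy also holds bottle and zipper bones, which the graph leaves untouched.
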ControlRigBlueprint /BRCosmetics/Accessories/FORT_Backpacks/Backpack_M_MED_FlakeSlide/Meshes/M_MED_FlakeSlide_Pack_CtrlRig
graphs [
 RigVMFunctionLibrary {links [], nodes []}
 RigVMModel {nodes [
   RigUnit_BeginExecution = &RigUnit_BeginExecution()
   RigVMFunction_ControlFlowBranch = &RigVMFunction_ControlFlowBranch(ExecuteContext=@RigUnit_BeginExecution.ExecutePin, Condition=RigVMVariableNode() {variable=skydive}.Value)
   &RigUnit_ModifyTransforms(
    ItemToModify=((Item=(Type=Bone,Name="pack"),Transform=(Rotation=(X=-0.1363,Y=0.0108,Z=-0.078,W=0.9875),Translation=(X=0,Y=1.6055,Z=-0.4811),Scale3D=(X=1,Y=1,Z=1))))
    Weight=1
    WeightMinimum=0
    ExecutePin=@RigVMFunction_ControlFlowBranch.False
   )
   &RigUnit_ModifyTransforms(
    ExecutePin=()
    ItemToModify=((Item=(Type=Bone,Name="pack"),Transform=(Rotation=(X=-0.1363,Y=0.0108,Z=-0.078,W=0.9875),Translation=(X=0,Y=0,Z=0),Scale3D=(X=1,Y=1,Z=1))))
    Weight=1
    WeightMinimum=0
   )
  ]}
]
hierarchy [
 root {loc=<0 0 0>, rot=<0 0 0 1>}
 pack {parent=root, loc=<0 -0.7316 0>, rot=<0 0 0 1>}
 base_bottle_01 {parent=pack, loc=<-4.2032 -13.6492 -14.1245>, rot=<0.2629 -0.7176 0.6096 0.2108>}
 dyn_bottle_01 {parent=base_bottle_01, loc=<0 0 0.3544>, rot=<0 0 0 1>}
 dyn_bottle_02 {parent=dyn_bottle_01, loc=<0 0 0>, rot=<0 0 0 1>}
 base_zip_02 {parent=pack, loc=<-7.4698 -7.7999 -1.4272>, rot=<-0.1982 -0.0883 0.9759 0.0233>}
 dyn_zip_01 {parent=base_zip_02, loc=<0 0 0>, rot=<0 0 0 1>}
]
variables [skydive {type=bool, default=False}]
