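# Anim Blueprint for the FlakeSlide backpack that adds secondary motion and adjusts the pack during skydiving.
# A Control Rig driven by dive mode feeds an additive root offset, weighted fully while diving, floating or parachuting.
# Spring-constrained AnimDynamics then simulate the zip and bottle bones in root-relative space.
# The event graph sets that weight and a front-end idle flag from the HighBeam tag and idle-bypass curve, which the AnimGraph never reads.
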AnimBlueprint /BRCosmetics/Accessories/FORT_Backpacks/Backpack_M_MED_FlakeSlide/Meshes/M_MED_FlakeSlide_Pack_AnimBP
target_skeleton=&/M_MED_FlakeSlide_Pack_Skeleton
graphs [
 AnimGraph {nodes [
   n1 = Control Rig!(Source=Input Pose(), Alpha, skydive:bool=Get bIsSkydiveDiveMode()) -> (Pose) {controlRigAssetReference {blueprintRigClass=&/M_MED_FlakeSlide_Pack_CtrlRig.M_MED_FlakeSlide_Pack_CtrlRig_C}}
   n2 = "Transform (Modify) Bone - Bone: root"(
    ComponentPose=Local To Component(LocalPose=@n1)
    Alpha=Get Alpha_skydiveMode() -> (Alpha_skydiveMode:float)
    Translation="0,1,0.026"
    Rotation="2.088,7.650,5.389"
   ) {
    rotationMode=BMM_Additive
    rotationSpace=BCS_BoneSpace
    translationMode=BMM_Additive
    translationSpace=BCS_BoneSpace
    alphaBoolBlend {blendOption=Linear}
    boneToModify {boneName=root}
    rotation=<rot 2.088 7.65 5.389>
    translation {y=1, z=0.026}
   }
   n3 = Anim Dynamics - Bone: dyn_zip_01(ComponentPose=@n2, Alpha=1, ExternalForce="0,0,0") {
    angularDampingOverride=1
    angularSpringConstant=3000
    bAngularSpring=true
    bOverrideAngularDamping=true
    bUseGravityOverride=true
    gravityScale=0.35
    simulationSpace=RootRelative
    alphaBoolBlend {blendOption=Linear}
    boundBone {boneName=dyn_zip_01}
    componentAppliedLinearAccClamp=<3000 3000 3000>
    componentLinearAccScale=<2 2 2>
    componentLinearVelScale=<2 2 2>
    gravityOverride {z=-980}
    physicsBodyDefinitions [{
      boundBone {boneName=dyn_zip_01}
      boxExtents=<20 30 20>
      constraintSetup {angularLimitsMax=<0 10 10>, angularLimitsMin=<0 -100 -10>, angularTarget=<1 0 0>}
      localJointOffset=<5 0 0>
     }]
   }
   n4 = Anim Dynamics - Bone: dyn_bottle_01(ComponentPose=@n3, Alpha=1, ExternalForce="0,0,0") {
    angularDampingOverride=1
    angularSpringConstant=15000
    bAngularSpring=true
    bOverrideAngularDamping=true
    bUseGravityOverride=true
    gravityScale=0.35
    simulationSpace=RootRelative
    alphaBoolBlend {blendOption=Linear}
    boundBone {boneName=dyn_bottle_01}
    componentAppliedLinearAccClamp=<2000 2000 2000>
    componentLinearAccScale=<1 1 1>
    componentLinearVelScale=<1 1 1>
    gravityOverride {z=-980}
    physicsBodyDefinitions [{
      boundBone {boneName=dyn_bottle_01}
      boxExtents=<30 30 30>
      constraintSetup {angularLimitsMax=<10 0 10>, angularLimitsMin=<-10 0 -10>, angularTarget=<1 0 0>}
      localJointOffset=<0 2 0>
     }]
   }
   Output Pose(Result=Component To Local(ComponentPose=@n4)) {layerGroup=None}
  ]}
 EventGraph {nodes [
   Event BlueprintUpdateAnimation()
   > n1 = Sequence()
   > n2 = Branch(Condition=OR Boolean!(A=Get bIsSkydiveDiveMode(), B=Get bIsSkydiveFloating(), C:bool=Get bIsParachuteOpened()) -> (ReturnValue))
   > Set Alpha_skydiveMode(Alpha_skydiveMode:float=1) -> (Output_Get:float)
   Set Alpha_skydiveMode(execute=@n2.else, Alpha_skydiveMode=0)
   Branch(execute=@n1.then_1, Condition=Get bIsInFrontEnd())
   n3 = Make Array("[0]"=(TagName="Cosmetics.MetaTags.HighBeam"))
   n4 = GetCurveValue(self=GetOwnerFrontEndAnimInstance(), CurveName=FrontEnd_Idle_Bypass)
   > n5 = Branch(Condition=AND Boolean(A=CharacterItemDefinitionHasAnyMetaTag(Tags=@n3), B="float >= float"(A=@n4, B=1)))
   > Set FE Idle Active(FE Idle Active:bool=true) -> (Output_Get:bool)
   Set FE Idle Active(execute=@n5.else, FE Idle Active=false)
  ]}
]
variables [
 Alpha_skydiveMode {type=number}
 FE Idle Active {type=boolean}
]
